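# Unlit, two-sided translucent material for the Battle Suit pickaxe's animated trail, flagged for beam trails and Niagara ribbons.
# A panning normal map, flattened by the UVDistortion dynamic parameter, warps UVs for three fire-mask layers, each bent along V by a sine.
# Their sum indexes the BattleSuit_Blue curve atlas row, tinted by particle color and HDR-boosted.
# Opacity combines that mask with U-edge falloffs, near-camera fade, particle alpha and non-mobile depth fade.
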Material /BRCosmetics/Weapons/FORT_Melee/Pickaxe_BattleSuit/Meshes/FX/M_Battlesuit_AnimTrail
graph {
 properties {
  bUsedWithBeamTrails=true
  bUsedWithNiagaraRibbons=true
  blendMode=BLEND_Translucent
  pixelDepthOffsetMode=PDOM_Legacy
  shadingModel=MSM_Unlit
  twoSided=true
  usedShadingModels=Unlit
 }
 nodes [
  n1 = TextureSample(UVs=Panner(Coordinate=TextureCoordinate()) {speedX=-1, speedY=0.1}) {
   samplerSource=SSM_Wrap_WorldGroupSettings
   samplerType=SAMPLERTYPE_Normal
   texture=/Game/Effects/Fort_Effects/Gameplay/SmokeDustExplosions/Textures/T_Explosion_NM_01
  }
  n2 = ComponentMask(None=MaterialFunctionCall(Normal=@n1.RGB, Flatness=DynamicParameter() {defaultValue=<rgba 0.975 1 1 1>, paramNames [UVDistortion,Param2,Param3,Param4]}.UVDistortion) {materialFunction=&FlattenNormal}.Result) {g=true, r=true}
  n3 = Add(A=TextureCoordinate(), B=@n2)
  n4 = OneMinus(None=ComponentMask(None=@n3) {r=true})
  n5 = OneMinus(None=ComponentMask(None=@n3) {r=true})
  n6 = TextureCoordinate()
  n7 = LinearInterpolate(
   A=AppendVector(A=ComponentMask(None=@n3) {g=true}, B=Add(A=@n4) {constB=-0.15})
   B=AppendVector(A=ComponentMask(None=@n3) {g=true}, B=Add(A=@n5) {constB=0.45})
   Alpha=Sine(None=ComponentMask(None=@n6) {g=true}) {period=2}
  )
  n8 = ScalarParameter() {defaultValue=1.5, parameterName=Bias_A}
  n9 = Add(A=TextureCoordinate(), B=@n2)
  n10 = OneMinus(None=ComponentMask(None=@n9) {r=true})
  n11 = OneMinus(None=ComponentMask(None=@n9) {r=true})
  n12 = TextureCoordinate()
  n13 = LinearInterpolate(
   A=AppendVector(A=ComponentMask(None=@n9) {g=true}, B=Add(A=@n10) {constB=-0.15})
   B=AppendVector(A=ComponentMask(None=@n9) {g=true}, B=Add(A=@n11) {constB=0.45})
   Alpha=Sine(None=ComponentMask(None=@n12) {g=true}) {period=2}
  )
  n14 = Multiply(A=Multiply(
    A=TextureSampleParameter2D!(UVs=@n7, Bias=@n8, Apply View MipBias) {mipValueMode=TMVM_MipBias, parameterName=Texture2D, texture=/Game/Effects/Fort_Effects/Textures/Fire/T_Fire_Mask}.R
    B=TextureSampleParameter2D(UVs=@n13, Bias=@n8) {mipValueMode=TMVM_MipBias, parameterName=Texture2D, texture=&T_Fire_Mask}.R
   )) {constB=0.333}
  n15 = Add(A=TextureCoordinate(), B=@n2)
  n16 = OneMinus(None=ComponentMask(None=@n15) {r=true})
  n17 = OneMinus(None=ComponentMask(None=@n15) {r=true})
  n18 = TextureCoordinate()
  n19 = LinearInterpolate(
   A=AppendVector(A=ComponentMask(None=@n15) {g=true}, B=Add(A=@n16) {constB=-0.15})
   B=AppendVector(A=ComponentMask(None=@n15) {g=true}, B=Add(A=@n17) {constB=0.25})
   Alpha=Sine(None=ComponentMask(None=@n18) {g=true}) {period=2}
  )
  n20 = ScalarParameter() {defaultValue=1.5, parameterName=Bias_A}
  n21 = Add(A=TextureCoordinate(), B=@n2)
  n22 = OneMinus(None=ComponentMask(None=@n21) {r=true})
  n23 = OneMinus(None=ComponentMask(None=@n21) {r=true})
  n24 = TextureCoordinate()
  n25 = LinearInterpolate(
   A=AppendVector(A=ComponentMask(None=@n21) {g=true}, B=Add(A=@n22) {constB=-0.15})
   B=AppendVector(A=ComponentMask(None=@n21) {g=true}, B=Add(A=@n23) {constB=0.25})
   Alpha=Sine(None=ComponentMask(None=@n24) {g=true}) {period=2}
  )
  n26 = Multiply(A=Multiply(
    A=TextureSampleParameter2D(UVs=@n19, Bias=@n20) {mipValueMode=TMVM_MipBias, parameterName=Texture2D, texture=&T_Fire_Mask}.R
    B=TextureSampleParameter2D(UVs=@n25, Bias=@n20) {mipValueMode=TMVM_MipBias, parameterName=Texture2D, texture=&T_Fire_Mask}.R
   )) {constB=0.333}
  n27 = Add(A=TextureCoordinate(), B=@n2)
  n28 = ComponentMask(None=@n27) {r=true}
  n29 = AppendVector(A=ComponentMask(None=@n27) {g=true}, B=Add(A=OneMinus(None=@n28)) {constB=-0.15})
  n30 = ComponentMask(None=@n27) {r=true}
  n31 = AppendVector(A=ComponentMask(None=@n27) {g=true}, B=Add(A=OneMinus(None=@n30)) {constB=0})
  n32 = Sine(None=ComponentMask(None=TextureCoordinate()) {g=true}) {period=2}
  n33 = ScalarParameter() {defaultValue=1.5, parameterName=Bias_A}
  n34 = Add(A=TextureCoordinate(), B=@n2)
  n35 = ComponentMask(None=@n34) {r=true}
  n36 = AppendVector(A=ComponentMask(None=@n34) {g=true}, B=Add(A=OneMinus(None=@n35)) {constB=-0.15})
  n37 = ComponentMask(None=@n34) {r=true}
  n38 = AppendVector(A=ComponentMask(None=@n34) {g=true}, B=Add(A=OneMinus(None=@n37)) {constB=0})
  n39 = Sine(None=ComponentMask(None=TextureCoordinate()) {g=true}) {period=2}
  n40 = Multiply(
   A=TextureSampleParameter2D(UVs=LinearInterpolate(A=@n29, B=@n31, Alpha=@n32), Bias=@n33) {mipValueMode=TMVM_MipBias, parameterName=Texture2D, texture=&T_Fire_Mask}.R
   B=TextureSampleParameter2D(UVs=LinearInterpolate(A=@n36, B=@n38, Alpha=@n39), Bias=@n33) {mipValueMode=TMVM_MipBias, parameterName=Texture2D, texture=&T_Fire_Mask}.R
  )
  n41 = Clamp(None=Add(A=Add(A=@n14, B=@n26), B=Multiply(A=@n40) {constB=0.333}))
  n42 = CurveAtlasRowParameter(CurveTime=@n41) {
   atlas=&AthenaCurveAtlas01
   curve=/Game/Characters/Player/Male/Medium/Bodies/M_MED_BattleSuit/FX/BattleSuit_Blue
   defaultValue=82
   parameterName=Curve
  }
  n43 = ParticleColor()
  n44 = Multiply(A=@n42, B=@n43.RGB)
  MaterialFunctionCall(Emissive=MaterialFunctionCall(LDR Input=Clamp(None=@n44), HDR Tint=@n44, HDR Intensity Multiplier=ScalarParameter() {defaultValue=1, parameterName=HDR Intensity}) {materialFunction=&DeriveHDRfromLDR}.RemappedRange) {materialFunction=&MF_FX_EmissiveAdjustments}
  n45 = Clamp(None=@n41)
  n46 = ComponentMask(None=TextureCoordinate()) {r=true}
  n47 = ScalarParameter() {defaultValue=1.5, parameterName=Leading Edge Exponent}
  n48 = Saturate(None=Multiply(A=Power(Base=@n46, Exp=@n47), B=ScalarParameter() {defaultValue=5, parameterName=Leading Edge Intensity}))
  n49 = ComponentMask(None=TextureCoordinate()) {r=true}
  n50 = Multiply(
   A=Multiply(A=Multiply(A=@n45, B=@n48), B=OneMinus(None=@n49))
   B=MaterialFunctionCall(Distance=ScalarParameter() {defaultValue=64, parameterName=NearFadeDistance}) {materialFunction=&Near_Camera_Fade}.Result
  )
  n51 = ScalarParameter() {defaultValue=150, parameterName=Depth Fade Distance}
  Multiply(A=@n43.A, B=ShadingPathSwitch(Default=DepthFade(Opacity=@n50, FadeDistance=@n51), Mobile=@n50) {desc=Migrated from FeatureLevelSwitch})
 ]
}
